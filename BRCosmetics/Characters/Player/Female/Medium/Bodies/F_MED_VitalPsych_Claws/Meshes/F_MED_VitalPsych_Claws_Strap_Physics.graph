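# Physics asset for the VitalPsych Claws medium female body cosmetic, covering the dynamic chest straps.
# Capsules on the upper spine, clavicles and upper arms form the body volume around four two-bone strap chains (front and back, left and right).
# Each chain hangs off spine_05 with per-strap swing limits and zero twist.
# Body joints have zero-degree limits, except the right upper arm, whose swing and twist are free.
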PhysicsAsset /BRCosmetics/Characters/Player/Female/Medium/Bodies/F_MED_VitalPsych_Claws/Meshes/F_MED_VitalPsych_Claws_Strap_Physics
bodies {
 clavicle_l [{shapeType=Capsule, length=10.5, radius=4.25, shapeName=clavicle_l_capsule, center=<6.0862 0 -0.5325>, rotation=<rot 85 -0.0023 -0.0023>}]
 clavicle_r [{shapeType=Capsule, length=10.5, radius=4.25, shapeName=clavicle_r_capsule, center=<-6.0862 0 0.5325>, rotation=<rot -85 179.9977 -179.9981>}]
 dyn_chestStrap_bck_l_01 [{
   shapeType=Capsule
   length=2
   radius=1
   shapeName=dyn_chestStrap_bck_l_01_capsule
   center=<1.4788 0 0.0828>
   rotation=<rot 84.9998 -0.0001 -0.0002>
  }]
 dyn_chestStrap_bck_l_02 [{
   shapeType=Capsule
   length=1.5
   radius=1
   shapeName=dyn_chestStrap_bck_l_02_capsule
   center=<-0.7077 0 0.0619>
   rotation=<rot 84.9998 -0.0001 -0.0002>
  }]
 dyn_chestStrap_bck_r_01 [{
   shapeType=Capsule
   length=2.25
   radius=1
   shapeName=dyn_chestStrap_bck_r_01_capsule
   center=<1.5305 0 -0.1339>
   rotation=<rot 84.9998 0.0002 0.0002>
  }]
 dyn_chestStrap_bck_r_02 [{
   shapeType=Capsule
   length=1.75
   radius=1
   shapeName=dyn_chestStrap_bck_r_02_capsule
   center=<-0.7077 0 0.0619>
   rotation=<rot 84.9998 0.0001 0.0002>
  }]
 dyn_chestStrap_fr_l_01 [{
   shapeType=Capsule
   length=2.25
   radius=1
   shapeName=dyn_chestStrap_fr_l_01_capsule
   center=<1.8086 0 -0.1582>
   rotation=<rot 85.0001 -0.0003 -0.0004>
  }]
 dyn_chestStrap_fr_l_02 [{
   shapeType=Capsule
   length=5.5
   radius=1
   shapeName=dyn_chestStrap_fr_l_02_capsule
   center=<2.7499 0 -0.2406>
   rotation=<rot 85.0002 0 -0.0001>
  }]
 dyn_chestStrap_fr_r_01 [{
   shapeType=Capsule
   length=2.25
   radius=1
   shapeName=dyn_chestStrap_fr_r_01_capsule
   center=<1.7903 0 0.1666>
   rotation=<rot 84.9999 179.9997 179.9999>
  }]
 dyn_chestStrap_fr_r_02 [{
   shapeType=Capsule
   length=4.5
   radius=1
   shapeName=dyn_chestStrap_fr_r_02_capsule
   center=<3.2684 0 -0.2859>
   rotation=<rot 85.0001 0.0001 0.0003>
  }]
 spine_04 [
  {shapeType=Capsule, length=5.75, radius=7.25, shapeName=spine_04_capsule, rotation=<rot 0 -0.0005 0>}
  {shapeType=Capsule, length=8.25, radius=5.25, shapeName=spine_04_capsule, center=<3.0014 5.5198 0>}
 ]
 spine_05 [{shapeType=Capsule, length=9.5, radius=7.5, shapeName=spine_05_capsule, center=<0 0.3062 0>, rotation=<rot 0 -0.0005 0>}]
 upperarm_l [{shapeType=Capsule, length=12, radius=4.5, shapeName=upperarm_l_capsule, center=<10.3677 0.0001 -0.2882>, rotation=<rot -90 0.1189 -0.1183>}]
 upperarm_r [{shapeType=Capsule, length=10.25, radius=3.005, shapeName=upperarm_r_capsule, center=<-12.5453 0 0>, rotation=<rot 0 90 90>}]
}
constraints [
 {bone1Name=spine_05, bone2Name=spine_04, swing1LimitDegrees=0, swing2LimitDegrees=0, twistLimitDegrees=0}
 {bone1Name=clavicle_l, bone2Name=spine_05, swing1LimitDegrees=0, swing2LimitDegrees=0, twistLimitDegrees=0}
 {bone1Name=upperarm_l, bone2Name=clavicle_l, swing1LimitDegrees=0, swing2LimitDegrees=0, twistLimitDegrees=0}
 {bone1Name=clavicle_r, bone2Name=spine_05, swing1LimitDegrees=0, swing2LimitDegrees=0, twistLimitDegrees=0}
 {bone1Name=dyn_chestStrap_bck_l_01, bone2Name=spine_05, swing1LimitDegrees=35, swing2LimitDegrees=50, twistLimitDegrees=0}
 {bone1Name=dyn_chestStrap_bck_l_02, bone2Name=dyn_chestStrap_bck_l_01, swing1LimitDegrees=35, swing2LimitDegrees=50, twistLimitDegrees=0}
 {bone1Name=dyn_chestStrap_bck_r_01, bone2Name=spine_05, swing1LimitDegrees=45, swing2LimitDegrees=60, twistLimitDegrees=0}
 {bone1Name=dyn_chestStrap_bck_r_02, bone2Name=dyn_chestStrap_bck_r_01, swing1LimitDegrees=40, swing2LimitDegrees=55, twistLimitDegrees=0}
 {bone1Name=dyn_chestStrap_fr_l_01, bone2Name=spine_05, swing1LimitDegrees=45, swing2LimitDegrees=45, twistLimitDegrees=0}
 {bone1Name=dyn_chestStrap_fr_l_02, bone2Name=dyn_chestStrap_fr_l_01, swing1LimitDegrees=45, swing2LimitDegrees=45, twistLimitDegrees=0}
 {bone1Name=dyn_chestStrap_fr_r_01, bone2Name=spine_05, swing1LimitDegrees=15, swing2LimitDegrees=45, twistLimitDegrees=0}
 {bone1Name=dyn_chestStrap_fr_r_02, bone2Name=dyn_chestStrap_fr_r_01, swing1LimitDegrees=15, swing2LimitDegrees=45, twistLimitDegrees=0}
 {
  bone1Name=upperarm_r
  bone2Name=clavicle_r
  swing1LimitDegrees=45
  swing1Motion=Free
  swing2LimitDegrees=45
  swing2Motion=Free
  twistLimitDegrees=45
  twistMotion=Free
 }
]
