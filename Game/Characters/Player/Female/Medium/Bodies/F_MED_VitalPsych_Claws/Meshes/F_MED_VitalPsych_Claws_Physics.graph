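# Physics asset for the medium female VitalPsych Claws player body, set up for secondary motion of its fur collar.
# Capsule bodies on upper spine, neck, head, clavicles and upper arms are joined by constraints with zero-degree limits.
# Front collar chains hang off each clavicle and fork after the second bone, while three two-bone back chains hang off neck_01.
# Collar joints allow small swings with zero twist, tighter down the front-right chain than the left.
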PhysicsAsset /Game/Characters/Player/Female/Medium/Bodies/F_MED_VitalPsych_Claws/Meshes/F_MED_VitalPsych_Claws_Physics
bodies {
 clavicle_l [{shapeType=Capsule, length=4.5, radius=4, center=<9.7355 0 -0.7226>, rotation=<rot 90 -87.106 -87.1063>}]
 clavicle_r [{shapeType=Capsule, length=4.25, radius=4.25, center=<-9.4474 1.1945 0.5278>, rotation=<rot -85 0.0002 -0.0004>}]
 dyn_furCollar_bck_c_01 [{shapeType=Capsule, length=5, radius=4.5, center=<2.132 0.6105 -1.2176>, rotation=<rot 79.9997 0.0001 90.0001>}]
 dyn_furCollar_bck_c_02 [{shapeType=Capsule, length=1, radius=1.75, center=<1.0655 -0.2899 -0.1879>, rotation=<rot 79.9997 -0.0002 -80.0002>}]
 dyn_furCollar_bck_l_01 [{shapeType=Capsule, length=0.75, radius=4.75, center=<1.5344 0.6427 -1.4662>, rotation=<rot 79.9999 -0.0003 -20.0003>}]
 dyn_furCollar_bck_l_02 [{shapeType=Capsule, length=1, radius=1.75, rotation=<rot 79.9998 -0.0003 -105.0002>}]
 dyn_furCollar_bck_r_01 [{shapeType=Capsule, length=1.75, radius=4, center=<0.2178 0.2326 -1.1973>, rotation=<rot 40.6292 -7.7486 -14.1739>}]
 dyn_furCollar_bck_r_02 [{shapeType=Capsule, length=0.75, radius=2.5, center=<-0.1508 -0.8682 0.0266>, rotation=<rot 79.9998 0 -80>}]
 dyn_furCollar_fr_l_01 [{shapeType=Capsule, length=2.25, radius=3, center=<2.7106 0 0.3178>, rotation=<rot -84.9999 179.9985 135.0012>}]
 dyn_furCollar_fr_l_02 [{shapeType=Capsule, length=2, radius=3.25, center=<2.5582 0 -1.4242>, rotation=<rot -84.9999 179.9972 -89.9976>}]
 dyn_furCollar_fr_l_03 [{shapeType=Capsule, length=1, radius=4.25, center=<1.8255 0 -0.1597>, rotation=<rot -84.9999 179.9975 -89.9979>}]
 dyn_furCollar_fr_l_04 [{shapeType=Capsule, length=2.75, radius=2.5, center=<2.2396 -0.0371 -0.1959>, rotation=<rot -84.9999 179.9988 -84.999>}]
 dyn_furCollar_fr_r_01 [{shapeType=Capsule, length=2.25, radius=3, center=<2.8724 0 0.28>, rotation=<rot -84.9999 -179.9989 -125.0008>}]
 dyn_furCollar_fr_r_02 [{shapeType=Capsule, length=2.5, radius=3.5, center=<3.3602 0 -0.294>, rotation=<rot -84.9999 -179.9986 -90.0012>}]
 dyn_furCollar_fr_r_03 [{shapeType=Capsule, length=0.5, radius=4.75, center=<2.2825 0.1269 -1.4521>, rotation=<rot -34.5069 -26.3681 92.562>}]
 dyn_furCollar_fr_r_04 [{shapeType=Capsule, length=1.75, radius=3.5, center=<2.3497 -0.7095 -0.3691>, rotation=<rot -84.9995 -179.997 174.9971>}]
 head [{shapeType=Capsule, length=0, radius=11.25}]
 neck_01 [{shapeType=Capsule, length=3.5, radius=7.5, center=<0.3084 1.7556 0>, rotation=<rot -90 0 -15.0002>}]
 neck_02 [{shapeType=Capsule, length=0.25, radius=10.75, center=<3.416 0.6047 0>, rotation=<rot -90 0 -0.0002>}]
 spine_05 [{shapeType=Capsule, length=9.75, radius=8.75, center=<0 -2.0916 0>}]
 upperarm_l [{shapeType=Capsule, length=7.75, radius=5, center=<8.4697 0 -0.0477>, rotation=<rot -80 0.0015 -0.0012>}]
 upperarm_r [{shapeType=Capsule, length=7.5, radius=5.25, center=<-8.755 -0.0001 0.495>, rotation=<rot 80 -179.9984 -179.9989>}]
}
constraints [
 {bone1Name=clavicle_l, bone2Name=spine_05, swing1LimitDegrees=0, swing2LimitDegrees=0, twistLimitDegrees=0}
 {bone1Name=upperarm_l, bone2Name=clavicle_l, swing1LimitDegrees=0, swing2LimitDegrees=0, twistLimitDegrees=0}
 {bone1Name=clavicle_r, bone2Name=spine_05, swing1LimitDegrees=0, swing2LimitDegrees=0, twistLimitDegrees=0}
 {bone1Name=upperarm_r, bone2Name=clavicle_r, swing1LimitDegrees=0, swing2LimitDegrees=0, twistLimitDegrees=0}
 {bone1Name=dyn_furCollar_fr_l_01, bone2Name=clavicle_l, swing1LimitDegrees=5, swing2LimitDegrees=10, twistLimitDegrees=0}
 {bone1Name=dyn_furCollar_fr_l_02, bone2Name=dyn_furCollar_fr_l_01, swing1LimitDegrees=5, swing2LimitDegrees=15, twistLimitDegrees=0}
 {bone1Name=dyn_furCollar_fr_l_03, bone2Name=dyn_furCollar_fr_l_02, swing1LimitDegrees=5, swing2LimitDegrees=15, twistLimitDegrees=0}
 {bone1Name=dyn_furCollar_fr_l_04, bone2Name=dyn_furCollar_fr_l_02, swing1LimitDegrees=5, swing2LimitDegrees=15, twistLimitDegrees=0}
 {bone1Name=dyn_furCollar_fr_r_01, bone2Name=clavicle_r, swing1LimitDegrees=5, swing2LimitDegrees=10, twistLimitDegrees=0}
 {bone1Name=dyn_furCollar_fr_r_02, bone2Name=dyn_furCollar_fr_r_01, swing1LimitDegrees=1, swing2LimitDegrees=10, twistLimitDegrees=0}
 {bone1Name=dyn_furCollar_fr_r_03, bone2Name=dyn_furCollar_fr_r_02, swing1LimitDegrees=1, swing2LimitDegrees=15, twistLimitDegrees=0}
 {bone1Name=dyn_furCollar_fr_r_04, bone2Name=dyn_furCollar_fr_r_02, swing1LimitDegrees=1, swing2LimitDegrees=8, twistLimitDegrees=0}
 {bone1Name=dyn_furCollar_bck_c_01, bone2Name=neck_01, swing1LimitDegrees=5, swing2LimitDegrees=15, twistLimitDegrees=0}
 {bone1Name=dyn_furCollar_bck_c_02, bone2Name=dyn_furCollar_bck_c_01, swing1LimitDegrees=5, swing2LimitDegrees=20, twistLimitDegrees=0}
 {bone1Name=dyn_furCollar_bck_l_01, bone2Name=neck_01, swing1LimitDegrees=5, swing2LimitDegrees=15, twistLimitDegrees=0}
 {bone1Name=dyn_furCollar_bck_l_02, bone2Name=dyn_furCollar_bck_l_01, swing1LimitDegrees=5, swing2LimitDegrees=20, twistLimitDegrees=0}
 {bone1Name=dyn_furCollar_bck_r_01, bone2Name=neck_01, swing1LimitDegrees=5, swing2LimitDegrees=10, twistLimitDegrees=0}
 {bone1Name=dyn_furCollar_bck_r_02, bone2Name=dyn_furCollar_bck_r_01, swing1LimitDegrees=5, swing2LimitDegrees=20, twistLimitDegrees=0}
 {bone1Name=neck_01, bone2Name=spine_05, swing1LimitDegrees=0, swing2LimitDegrees=0, twistLimitDegrees=0}
 {bone1Name=neck_02, bone2Name=neck_01, swing1LimitDegrees=0, swing2LimitDegrees=0, twistLimitDegrees=0}
 {bone1Name=head, bone2Name=neck_02, swing1LimitDegrees=0, swing2LimitDegrees=0, twistLimitDegrees=0}
]
