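# Post-process-style Animation Blueprint for the female medium VitalPsych Claws body mesh, shaping and animating its fur collar.
# It copies the attached parent mesh's pose, then runs a deform Control Rig.
# Long stacks of additive Modify Bone offsets with scale overrides reshape the collar bones around per-bone AnimDynamics, simulated relative to spine_05 with overridden gravity.
# Low-alpha constraints after the dynamics pull front collar bones toward the clavicles and neck.
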AnimBlueprint /BRCosmetics/Characters/Player/Female/Medium/Bodies/F_MED_VitalPsych_Claws/Meshes/F_MED_VitalPsych_Claws_AnimBP
target_skeleton=&/F_MED_VitalPsych_Claws_Skeleton
graphs [
 AnimGraph {nodes [
   "Constraint - dyn_furCollar_fr_r_02 (PP)"!(ComponentPose, Alpha=0.5, ConstraintWeights_0=0.9, ConstraintWeights_1:float=0.1) -> (Pose) {
    alphaBoolBlend {blendOption=Linear}
    alphaScaleBiasClamp {bInterpResult=true}
    boneToModify {boneName=dyn_furCollar_fr_r_02}
    constraintSetup [{targetBone {boneName=clavicle_r}}, {targetBone {boneName=neck_01}}]
    constraintWeights [1,0]
   }
   RigidBody!(ComponentPose, Alpha, OverrideWorldGravity:Vector="0,0,0", ExternalForce) -> (Pose) {
    bClampLinearTranslationLimitToRefPose=true
    bOverrideWorldGravity=true
    overridePhysicsAsset=/Game/Characters/Player/Female/Medium/Bodies/F_MED_VitalPsych_Claws/Meshes/F_MED_VitalPsych_Claws_Physics
    simulationSpace=BaseBoneSpace
    alphaBoolBlend {blendOption=Linear}
    baseBoneRef {boneName=spine_05}
    componentLinearAccScale=<0.2 0.2 0.2>
    componentLinearVelScale=<0.2 0.2 0.2>
   }
   n1 = Control Rig(Source=Copy Pose From Mesh() {bUseAttachedParent=true}) {
    alphaBoolBlend {blendOption=Linear}
    controlRigAssetReference {blueprintRigClass=&/F_MED_DeformRig_ControlRig_VitalPsych_Claws.F_MED_DeformRig_ControlRig_VitalPsych_Claws_C}
   }
   n2 = "Transform (Modify) Bone - Bone: dyn_furCollar_bck_r_02"(ComponentPose=Local To Component(LocalPose=@n1), Alpha=1, Translation="0.000,-0.777,1.365", Rotation="0,0,0", Scale="1.000,1.000,0.750") {
    rotationMode=BMM_Additive
    scaleMode=BMM_Replace
    translationMode=BMM_Additive
    alphaBoolBlend {blendOption=Linear}
    boneToModify {boneName=dyn_furCollar_bck_r_02}
    scale {z=0.75}
    translation {y=-0.777, z=1.365}
   }
   n3 = "Transform (Modify) Bone - Bone: dyn_furCollar_fr_r_05"(ComponentPose=@n2, Alpha=1, Translation="0.000,-1.000,1.766", Rotation="-5.000,0.000,-0.000", Scale="1,1,0.8") {
    rotationMode=BMM_Additive
    scaleMode=BMM_Replace
    translationMode=BMM_Additive
    alphaBoolBlend {blendOption=Linear}
    boneToModify {boneName=dyn_furCollar_fr_r_05}
    rotation {pitch=-5}
    scale {z=0.8}
    translation {y=-1, z=1.766}
   }
   n4 = "Transform (Modify) Bone - Bone: dyn_furCollar_fr_r_base"(ComponentPose=@n3, Alpha=1, Translation="-1,-0.694,1.564", Rotation="0,0,0", Scale="0.750,1.000,0.750") {
    rotationMode=BMM_Additive
    scaleMode=BMM_Replace
    translationMode=BMM_Additive
    alphaBoolBlend {blendOption=Linear}
    alphaScaleBiasClamp {bInterpResult=true}
    boneToModify {boneName=dyn_furCollar_fr_r_base}
    scale {x=0.75, z=0.75}
    translation=<-1 -0.694 1.564>
   }
   n5 = "Transform (Modify) Bone - Bone: dyn_furCollar_fr_r_01"(ComponentPose=@n4, Alpha=1, Translation="1.046,2.045,0.000", Rotation="0,0,0", Scale="1,1,1") {
    rotationMode=BMM_Additive
    scaleMode=BMM_Replace
    translationMode=BMM_Additive
    alphaBoolBlend {blendOption=Linear}
    alphaScaleBiasClamp {bInterpResult=true}
    boneToModify {boneName=dyn_furCollar_fr_r_01}
    translation=<1.046 2.045>
   }
   n6 = "Transform (Modify) Bone - Bone: dyn_furCollar_bck_r_01"(ComponentPose=@n5, Alpha=1, Translation="0.874,-1.495,0.000", Rotation="0,0,0", Scale="0.85,1,1") {
    rotationMode=BMM_Additive
    scaleMode=BMM_Replace
    translationMode=BMM_Additive
    alphaBoolBlend {blendOption=Linear}
    alphaScaleBiasClamp {bInterpResult=true}
    boneToModify {boneName=dyn_furCollar_bck_r_01}
    scale {x=0.85}
    translation=<0.8736 -1.495>
   }
   n7 = "Transform (Modify) Bone - Bone: dyn_furCollar_bck_c_01"(ComponentPose=@n6, Alpha=1, Translation="-0.381,-3.292,-1.143", Rotation="0,0,0", Scale="1.000,1.000,0.750") {
    rotationMode=BMM_Additive
    scaleMode=BMM_Replace
    translationMode=BMM_Additive
    alphaBoolBlend {blendOption=Linear}
    alphaScaleBiasClamp {bInterpResult=true}
    boneToModify {boneName=dyn_furCollar_bck_c_01}
    scale {z=0.75}
    translation=<-0.381 -3.292 -1.143>
   }
   n8 = "Transform (Modify) Bone - Bone: dyn_furCollar_fr_r_base"(ComponentPose=@n7, Alpha=1, Translation="0.271,0.000,-0.429", Rotation="-29.497,-5.725,11.509", Scale="1,1,1") {
    rotationMode=BMM_Additive
    scaleMode=BMM_Replace
    translationMode=BMM_Additive
    alphaBoolBlend {blendOption=Linear}
    alphaScaleBiasClamp {bInterpResult=true}
    boneToModify {boneName=dyn_furCollar_fr_r_base}
    rotation=<rot -29.497 -5.725 11.509>
    translation {x=0.271, z=-0.429}
   }
   n9 = "Transform (Modify) Bone - Bone: dyn_furCollar_fr_r_01"(ComponentPose=@n8, Alpha=1, Translation="4.196,-0.000,3.938", Rotation="-39.384,-9.122,6.450", Scale="1,1,1") {
    rotationMode=BMM_Additive
    scaleMode=BMM_Replace
    translationMode=BMM_Additive
    alphaBoolBlend {blendOption=Linear}
    alphaScaleBiasClamp {bInterpResult=true}
    boneToModify {boneName=dyn_furCollar_fr_r_01}
    rotation=<rot -39.384 -9.122 6.45>
    translation {x=4.196, z=3.938}
   }
   n10 = "Transform (Modify) Bone - Bone: dyn_furCollar_fr_r_02"(ComponentPose=@n9, Alpha=1, Translation="0.319,-1.003,1.225", Rotation="0.000,0.000,-30.000", Scale="1,1,1") {
    rotationMode=BMM_Additive
    scaleMode=BMM_Replace
    translationMode=BMM_Additive
    alphaBoolBlend {blendOption=Linear}
    alphaScaleBiasClamp {bInterpResult=true}
    boneToModify {boneName=dyn_furCollar_fr_r_02}
    rotation {roll=-30}
    translation=<0.319 -1.003 1.225>
   }
   n11 = "Transform (Modify) Bone - Bone: dyn_furCollar_fr_r_03"(ComponentPose=@n10, Alpha=1, Translation="0.937,-2.793,-1.376", Rotation="18.058,-8.745,-26.392", Scale="0.750,1.000,1.000") {
    rotationMode=BMM_Additive
    scaleMode=BMM_Replace
    translationMode=BMM_Additive
    alphaBoolBlend {blendOption=Linear}
    alphaScaleBiasClamp {bInterpResult=true}
    boneToModify {boneName=dyn_furCollar_fr_r_03}
    rotation=<rot 18.058 -8.745 -26.392>
    scale {x=0.75}
    translation=<0.937 -2.7932 -1.376>
   }
   n12 = "Transform (Modify) Bone - Bone: dyn_furCollar_fr_r_04"(ComponentPose=@n11, Alpha=1, Translation="1.415,0.209,1.168", Rotation="28.878,-8.499,-22.193", Scale="1,1,1") {
    rotationMode=BMM_Additive
    scaleMode=BMM_Replace
    translationMode=BMM_Additive
    alphaBoolBlend {blendOption=Linear}
    alphaScaleBiasClamp {bInterpResult=true}
    boneToModify {boneName=dyn_furCollar_fr_r_04}
    rotation=<rot 28.878 -8.499 -22.193>
    translation=<1.415 0.209 1.168>
   }
   n13 = "Transform (Modify) Bone - Bone: dyn_furCollar_fr_r_05"(ComponentPose=@n12, Alpha=1, Translation="1.148,-0.875,1.506", Rotation="0.000,10.000,-0.000", Scale="1,1,0.5") {
    rotationMode=BMM_Additive
    scaleMode=BMM_Replace
    translationMode=BMM_Additive
    alphaBoolBlend {blendOption=Linear}
    alphaScaleBiasClamp {bInterpResult=true}
    boneToModify {boneName=dyn_furCollar_fr_r_05}
    rotation {yaw=10}
    scale {z=0.5}
    translation=<1.148 -0.875 1.506>
   }
   n14 = "Transform (Modify) Bone - Bone: dyn_furCollar_bck_r_01"(ComponentPose=@n13, Alpha=1, Translation="2.855,0.213,0.709", Rotation="0,0,0", Scale="1,1,1") {
    rotationMode=BMM_Additive
    scaleMode=BMM_Replace
    translationMode=BMM_Additive
    alphaBoolBlend {blendOption=Linear}
    alphaScaleBiasClamp {bInterpResult=true}
    boneToModify {boneName=dyn_furCollar_bck_r_01}
    translation=<2.855 0.213 0.7087>
   }
   n15 = "Transform (Modify) Bone - Bone: dyn_furCollar_fr_r_02"(ComponentPose=@n14, Alpha=1, Translation="2.824,-1.834,3.527", Rotation="0,0,0", Scale="1.000,0.500,1.000") {
    rotationMode=BMM_Additive
    scaleMode=BMM_Replace
    translationMode=BMM_Additive
    alphaBoolBlend {blendOption=Linear}
    alphaScaleBiasClamp {bInterpResult=true}
    boneToModify {boneName=dyn_furCollar_fr_r_02}
    scale {y=0.5}
    translation=<2.824 -1.834 3.527>
   }
   n16 = "Transform (Modify) Bone - Bone: dyn_furCollar_fr_r_01"(ComponentPose=@n15, Alpha=1, Translation="0,0,0", Rotation="0,0,0", Scale="1,1,1") {
    rotationMode=BMM_Additive
    scaleMode=BMM_Replace
    translationMode=BMM_Additive
    alphaBoolBlend {blendOption=Linear}
    alphaScaleBiasClamp {bInterpResult=true}
    boneToModify {boneName=dyn_furCollar_fr_r_01}
   }
   n17 = "Transform (Modify) Bone - Bone: dyn_furCollar_fr_r_03"(ComponentPose=@n16, Alpha=1, Translation="-0.000,0.000,3.570", Rotation="0,0,0", Scale="1,1,1") {
    rotationMode=BMM_Additive
    scaleMode=BMM_Replace
    translationMode=BMM_Additive
    alphaBoolBlend {blendOption=Linear}
    alphaScaleBiasClamp {bInterpResult=true}
    boneToModify {boneName=dyn_furCollar_fr_r_03}
    translation {z=3.57}
   }
   n18 = "Transform (Modify) Bone - Bone: dyn_furCollar_fr_r_04"(ComponentPose=@n17, Alpha=1, Translation="1.579,0.000,-0.605", Rotation="0,0,0", Scale="1,1,1") {
    rotationMode=BMM_Additive
    scaleMode=BMM_Replace
    translationMode=BMM_Additive
    alphaBoolBlend {blendOption=Linear}
    alphaScaleBiasClamp {bInterpResult=true}
    boneToModify {boneName=dyn_furCollar_fr_r_04}
    translation {x=1.579, z=-0.605}
   }
   n19 = "Transform (Modify) Bone - Bone: dyn_furCollar_fr_r_05"(ComponentPose=@n18, Alpha=1, Translation="0,0,0", Rotation="0,0,0", Scale="1,1,1") {
    rotationMode=BMM_Additive
    scaleMode=BMM_Replace
    translationMode=BMM_Additive
    alphaBoolBlend {blendOption=Linear}
    alphaScaleBiasClamp {bInterpResult=true}
    boneToModify {boneName=dyn_furCollar_fr_r_05}
   }
   n20 = "Transform (Modify) Bone - Bone: dyn_furCollar_fr_r_01"(ComponentPose=@n19, Alpha=1, Translation="-1.630,0.000,-0.000", Rotation="0,0,0", Scale="1,1,1") {
    rotationMode=BMM_Additive
    scaleMode=BMM_Replace
    translationMode=BMM_Additive
    alphaBoolBlend {blendOption=Linear}
    alphaScaleBiasClamp {bInterpResult=true}
    boneToModify {boneName=dyn_furCollar_fr_r_01}
    translation {x=-1.6296, z=0}
   }
   n21 = "Transform (Modify) Bone - Bone: dyn_furCollar_fr_l_base"(ComponentPose=@n20, Alpha=1, Translation="-1.309,-0.750,5.330", Rotation="16.880,-11.970,-57.665", Scale="1,1,1") {
    rotationMode=BMM_Additive
    scaleMode=BMM_Replace
    translationMode=BMM_Additive
    alphaBoolBlend {blendOption=Linear}
    alphaScaleBiasClamp {bInterpResult=true}
    boneToModify {boneName=dyn_furCollar_fr_l_base}
    rotation=<rot 16.88 -11.97 -57.665>
    translation=<-1.309 -0.75 5.3295>
   }
   n22 = "Transform (Modify) Bone - Bone: dyn_furCollar_fr_l_01"(ComponentPose=@n21, Alpha=1, Translation="0.406,-0.000,0.030", Rotation="0,0,0", Scale="1,1,1") {
    rotationMode=BMM_Additive
    scaleMode=BMM_Replace
    translationMode=BMM_Additive
    alphaBoolBlend {blendOption=Linear}
    alphaScaleBiasClamp {bInterpResult=true}
    boneToModify {boneName=dyn_furCollar_fr_l_01}
    translation {x=0.406, z=0.03}
   }
   n23 = "Transform (Modify) Bone - Bone: dyn_furCollar_fr_l_04"(ComponentPose=@n22, Alpha=1, Translation="0.000,0.713,0.927", Rotation="0.000,5.000,-20.000", Scale="1,1,1") {
    rotationMode=BMM_Additive
    scaleMode=BMM_Replace
    translationMode=BMM_Additive
    alphaBoolBlend {blendOption=Linear}
    alphaScaleBiasClamp {bInterpResult=true}
    boneToModify {boneName=dyn_furCollar_fr_l_04}
    rotation {roll=-20, yaw=5}
    translation {y=0.713, z=0.927}
   }
   n24 = "Transform (Modify) Bone - Bone: dyn_furCollar_fr_l_02"(ComponentPose=@n23, Alpha=1, Translation="0.000,0.666,-1.046", Rotation="0,0,0", Scale="1,1,1") {
    rotationMode=BMM_Additive
    scaleMode=BMM_Replace
    translationMode=BMM_Additive
    alphaBoolBlend {blendOption=Linear}
    alphaScaleBiasClamp {bInterpResult=true}
    boneToModify {boneName=dyn_furCollar_fr_l_02}
    translation {y=0.666, z=-1.046}
   }
   n25 = "Transform (Modify) Bone - Bone: dyn_furCollar_fr_l_03"(ComponentPose=@n24, Alpha=1, Translation="0.000,0.000,-0.478", Rotation="0,0,0", Scale="1,1,1") {
    rotationMode=BMM_Additive
    scaleMode=BMM_Replace
    translationMode=BMM_Additive
    alphaBoolBlend {blendOption=Linear}
    alphaScaleBiasClamp {bInterpResult=true}
    boneToModify {boneName=dyn_furCollar_fr_l_03}
    translation=<0 0 -0.4776>
   }
   n26 = "Transform (Modify) Bone - Bone: dyn_furCollar_fr_r_03"(ComponentPose=@n25, Alpha=1, Translation="-0.000,-0.000,0.663", Rotation="0,0,0", Scale="1,1,1") {
    rotationMode=BMM_Additive
    scaleMode=BMM_Replace
    translationMode=BMM_Additive
    alphaBoolBlend {blendOption=Linear}
    alphaScaleBiasClamp {bInterpResult=true}
    boneToModify {boneName=dyn_furCollar_fr_r_03}
    translation {z=0.663}
   }
   n27 = "Transform (Modify) Bone - Bone: dyn_furCollar_fr_r_04"(ComponentPose=@n26, Alpha=1, Translation="0.000,0.000,2.687", Rotation="0,0,0", Scale="1,1,1") {
    rotationMode=BMM_Additive
    scaleMode=BMM_Replace
    translationMode=BMM_Additive
    alphaBoolBlend {blendOption=Linear}
    alphaScaleBiasClamp {bInterpResult=true}
    boneToModify {boneName=dyn_furCollar_fr_r_04}
    translation {z=2.687}
   }
   n28 = "Transform (Modify) Bone - Bone: dyn_furCollar_fr_r_05"(ComponentPose=@n27, Alpha=1, Translation="-1.457,-3.309,1", Rotation="0,0,0", Scale="1,1,1") {
    rotationMode=BMM_Additive
    scaleMode=BMM_Replace
    translationMode=BMM_Additive
    alphaBoolBlend {blendOption=Linear}
    alphaScaleBiasClamp {bInterpResult=true}
    boneToModify {boneName=dyn_furCollar_fr_r_05}
    translation=<-1.457 -3.309 1>
   }
   n29 = "Transform (Modify) Bone - Bone: dyn_furCollar_bck_r_01"(ComponentPose=@n28, Alpha=1, Translation="-3.037,0.000,0.993", Rotation="0,0,0", Scale="1,1,1") {
    rotationMode=BMM_Additive
    scaleMode=BMM_Replace
    translationMode=BMM_Additive
    alphaBoolBlend {blendOption=Linear}
    alphaScaleBiasClamp {bInterpResult=true}
    boneToModify {boneName=dyn_furCollar_bck_r_01}
    translation {x=-3.037, z=0.993}
   }
   n30 = "Transform (Modify) Bone - Bone: dyn_furCollar_bck_r_01"(ComponentPose=@n29, Alpha=1, Translation="0,-1.4,0.41", Rotation="0,0,0", Scale="1,1,1") {
    rotationMode=BMM_Additive
    scaleMode=BMM_Replace
    translationMode=BMM_Additive
    alphaBoolBlend {blendOption=Linear}
    alphaScaleBiasClamp {bInterpResult=true}
    boneToModify {boneName=dyn_furCollar_bck_r_01}
    translation {y=-1.4, z=0.41}
   }
   n31 = "Transform (Modify) Bone - Bone: dyn_furCollar_bck_c_01"(ComponentPose=@n30, Alpha=1, Translation="0.000,-0.833,0.705", Rotation="0,0,0", Scale="1,1,1") {
    rotationMode=BMM_Additive
    scaleMode=BMM_Replace
    translationMode=BMM_Additive
    alphaBoolBlend {blendOption=Linear}
    alphaScaleBiasClamp {bInterpResult=true}
    boneToModify {boneName=dyn_furCollar_bck_c_01}
    translation {y=-0.833, z=0.705}
   }
   n32 = "Transform (Modify) Bone - Bone: dyn_furCollar_bck_l_01"(ComponentPose=@n31, Alpha=1, Translation="-0.000,-1.164,0.645", Rotation="0,0,0", Scale="1,1,1") {
    rotationMode=BMM_Additive
    scaleMode=BMM_Replace
    translationMode=BMM_Additive
    alphaBoolBlend {blendOption=Linear}
    alphaScaleBiasClamp {bInterpResult=true}
    boneToModify {boneName=dyn_furCollar_bck_l_01}
    translation {y=-1.164, z=0.645}
   }
   n33 = Anim Dynamics - Bone: dyn_furCollar_fr_l_01!(ComponentPose=@n32, Alpha=1, GravityOverride:Vector="0,0,0", ExternalForce="0,0,0") -> (Pose) {
    bGravityOverrideInSimSpace=true
    bUseGravityOverride=true
    alphaBoolBlend {blendOption=Linear}
    boundBone {boneName=dyn_furCollar_fr_l_01}
    componentAppliedLinearAccClamp=<10000 10000 10000>
    componentLinearAccScale=<0.4 0.4 0.4>
    physicsBodyDefinitions [{
      boundBone {boneName=dyn_furCollar_fr_l_01}
      boxExtents=<6 6 6>
      constraintSetup {angularLimitsMax=<20 20 20>, angularLimitsMin=<-10 -10 -20>, angularTarget=<0 0 0>}
      localJointOffset=<0 0.01 0>
     }]
    relativeSpaceBone {boneName=spine_05}
   }
   n34 = Anim Dynamics - Bone: dyn_furCollar_fr_l_03!(ComponentPose=@n33, Alpha=1, GravityOverride:Vector="0,0,0", ExternalForce="0,0,0") -> (Pose) {
    bGravityOverrideInSimSpace=true
    bUseGravityOverride=true
    alphaBoolBlend {blendOption=Linear}
    boundBone {boneName=dyn_furCollar_fr_l_03}
    componentAppliedLinearAccClamp=<10000 10000 10000>
    componentLinearAccScale=<0.4 0.4 0.4>
    physicsBodyDefinitions [{
      boundBone {boneName=dyn_furCollar_fr_l_03}
      boxExtents=<6 6 6>
      constraintSetup {angularLimitsMax=<20 20 20>, angularLimitsMin=<-10 -10 -20>, angularTarget=<0 0 0>}
      localJointOffset=<0 0.01 0>
     }]
    relativeSpaceBone {boneName=spine_05}
   }
   n35 = Anim Dynamics - Bone: dyn_furCollar_fr_l_04!(ComponentPose=@n34, Alpha=1, GravityOverride:Vector="0,0,0", ExternalForce="0,0,0") -> (Pose) {
    bGravityOverrideInSimSpace=true
    bUseGravityOverride=true
    alphaBoolBlend {blendOption=Linear}
    boundBone {boneName=dyn_furCollar_fr_l_04}
    componentAppliedLinearAccClamp=<10000 10000 10000>
    componentLinearAccScale=<0.4 0.4 0.4>
    physicsBodyDefinitions [{
      boundBone {boneName=dyn_furCollar_fr_l_04}
      boxExtents=<6 6 6>
      constraintSetup {angularLimitsMax=<20 20 20>, angularLimitsMin=<-10 -10 -20>, angularTarget=<0 0 0>}
      localJointOffset=<0 0.01 0>
     }]
    relativeSpaceBone {boneName=spine_05}
   }
   n36 = Anim Dynamics - Bone: dyn_furCollar_fr_r_01!(ComponentPose=@n35, Alpha=1, GravityOverride:Vector="0,0,0", ExternalForce="0,0,0") -> (Pose) {
    bGravityOverrideInSimSpace=true
    bUseGravityOverride=true
    alphaBoolBlend {blendOption=Linear}
    boundBone {boneName=dyn_furCollar_fr_r_01}
    componentAppliedLinearAccClamp=<10000 10000 10000>
    componentLinearAccScale=<0.4 0.4 0.4>
    physicsBodyDefinitions [{
      boundBone {boneName=dyn_furCollar_fr_r_01}
      boxExtents=<6 6 6>
      constraintSetup {angularLimitsMax=<20 20 20>, angularLimitsMin=<-10 -10 -20>, angularTarget=<0 0 0>}
      localJointOffset=<0 0.01 0>
     }]
    relativeSpaceBone {boneName=spine_05}
   }
   n37 = Anim Dynamics - Bone: dyn_furCollar_fr_r_03!(ComponentPose=@n36, Alpha=1, GravityOverride:Vector="0,0,0", ExternalForce="0,0,0") -> (Pose) {
    bGravityOverrideInSimSpace=true
    bUseGravityOverride=true
    alphaBoolBlend {blendOption=Linear}
    boundBone {boneName=dyn_furCollar_fr_r_03}
    componentAppliedLinearAccClamp=<10000 10000 10000>
    componentLinearAccScale=<0.4 0.4 0.4>
    physicsBodyDefinitions [{
      boundBone {boneName=dyn_furCollar_fr_r_03}
      boxExtents=<6 6 6>
      constraintSetup {angularLimitsMax=<20 20 20>, angularLimitsMin=<-10 -10 -20>, angularTarget=<0 0 0>}
      localJointOffset=<0 0.01 0>
     }]
    relativeSpaceBone {boneName=spine_05}
   }
   n38 = Anim Dynamics - Bone: dyn_furCollar_fr_r_04!(ComponentPose=@n37, Alpha=1, GravityOverride:Vector="0,0,0", ExternalForce="0,0,0") -> (Pose) {
    bGravityOverrideInSimSpace=true
    bUseGravityOverride=true
    alphaBoolBlend {blendOption=Linear}
    boundBone {boneName=dyn_furCollar_fr_r_04}
    componentAppliedLinearAccClamp=<10000 10000 10000>
    componentLinearAccScale=<0.4 0.4 0.4>
    physicsBodyDefinitions [{
      boundBone {boneName=dyn_furCollar_fr_r_04}
      boxExtents=<6 6 6>
      constraintSetup {angularLimitsMax=<20 20 20>, angularLimitsMin=<-10 -10 -20>, angularTarget=<0 0 0>}
      localJointOffset=<0 0.01 0>
     }]
    relativeSpaceBone {boneName=spine_05}
   }
   n39 = Anim Dynamics - Bone: dyn_furCollar_fr_r_05!(ComponentPose=@n38, Alpha=1, GravityOverride:Vector="0,0,0", ExternalForce="0,0,0") -> (Pose) {
    bGravityOverrideInSimSpace=true
    bUseGravityOverride=true
    alphaBoolBlend {blendOption=Linear}
    boundBone {boneName=dyn_furCollar_fr_r_05}
    componentAppliedLinearAccClamp=<10000 10000 10000>
    componentLinearAccScale=<0.4 0.4 0.4>
    physicsBodyDefinitions [{
      boundBone {boneName=dyn_furCollar_fr_r_05}
      boxExtents=<6 6 6>
      constraintSetup {angularLimitsMax=<20 20 20>, angularLimitsMin=<-10 -10 -20>, angularTarget=<0 0 0>}
      localJointOffset=<0 0.01 0>
     }]
    relativeSpaceBone {boneName=spine_05}
   }
   n40 = Anim Dynamics - Bone: dyn_furCollar_bck_l_02!(ComponentPose=@n39, Alpha=1, GravityOverride:Vector="0,0,0", ExternalForce="0,0,0") -> (Pose) {
    bGravityOverrideInSimSpace=true
    bUseGravityOverride=true
    alphaBoolBlend {blendOption=Linear}
    boundBone {boneName=dyn_furCollar_bck_l_02}
    componentAppliedLinearAccClamp=<10000 10000 10000>
    componentLinearAccScale=<0.4 0.4 0.4>
    physicsBodyDefinitions [{
      boundBone {boneName=dyn_furCollar_bck_l_02}
      boxExtents=<6 6 6>
      constraintSetup {angularLimitsMax=<20 20 20>, angularLimitsMin=<-10 -10 -20>, angularTarget=<0 0 0>}
      localJointOffset=<0 0.01 0>
     }]
    relativeSpaceBone {boneName=spine_05}
   }
   n41 = Anim Dynamics - Bone: dyn_furCollar_bck_r_02!(ComponentPose=@n40, Alpha=1, GravityOverride:Vector="0,0,0", ExternalForce="0,0,0") -> (Pose) {
    bGravityOverrideInSimSpace=true
    bUseGravityOverride=true
    alphaBoolBlend {blendOption=Linear}
    boundBone {boneName=dyn_furCollar_bck_r_02}
    componentAppliedLinearAccClamp=<10000 10000 10000>
    componentLinearAccScale=<0.4 0.4 0.4>
    physicsBodyDefinitions [{
      boundBone {boneName=dyn_furCollar_bck_r_02}
      boxExtents=<6 6 6>
      constraintSetup {angularLimitsMax=<20 20 20>, angularLimitsMin=<-10 -10 -20>, angularTarget=<0 0 0>}
      localJointOffset=<0 0.01 0>
     }]
    relativeSpaceBone {boneName=spine_05}
   }
   n42 = Anim Dynamics - Bone: dyn_furCollar_bck_c_02!(ComponentPose=@n41, Alpha=1, GravityOverride:Vector="0,0,0", ExternalForce="0,0,0") -> (Pose) {
    bGravityOverrideInSimSpace=true
    bUseGravityOverride=true
    alphaBoolBlend {blendOption=Linear}
    boundBone {boneName=dyn_furCollar_bck_c_02}
    componentAppliedLinearAccClamp=<10000 10000 10000>
    componentLinearAccScale=<0.4 0.4 0.4>
    physicsBodyDefinitions [{
      boundBone {boneName=dyn_furCollar_bck_c_02}
      boxExtents=<6 6 6>
      constraintSetup {angularLimitsMax=<20 20 20>, angularLimitsMin=<-10 -10 -20>, angularTarget=<0 0 0>}
      localJointOffset=<0 0.01 0>
     }]
    relativeSpaceBone {boneName=spine_05}
   }
   n43 = "Constraint - dyn_furCollar_fr_r_base (PP)"!(ComponentPose=@n42, Alpha=0.1, ConstraintWeights_0=0.9, ConstraintWeights_1:float=0.1) -> (Pose) {
    alphaBoolBlend {blendOption=Linear}
    alphaScaleBiasClamp {bInterpResult=true}
    boneToModify {boneName=dyn_furCollar_fr_r_base}
    constraintSetup [{targetBone {boneName=clavicle_r}}, {targetBone {boneName=neck_01}}]
    constraintWeights [1,0]
   }
   n44 = "Constraint - dyn_furCollar_fr_l_base (PP)"!(ComponentPose=@n43, Alpha=0.1, ConstraintWeights_0=0.9, ConstraintWeights_1:float=0.1) -> (Pose) {
    alphaBoolBlend {blendOption=Linear}
    alphaScaleBiasClamp {bInterpResult=true}
    boneToModify {boneName=dyn_furCollar_fr_l_base}
    constraintSetup [{targetBone {boneName=clavicle_l}}, {targetBone {boneName=neck_01}}]
    constraintWeights [1,0]
   }
   n45 = "Constraint - dyn_furCollar_fr_r_01 (PP)"!(ComponentPose=@n44, Alpha=0.1, ConstraintWeights_0=0.9, ConstraintWeights_1:float=0.1) -> (Pose) {
    alphaBoolBlend {blendOption=Linear}
    alphaScaleBiasClamp {bInterpResult=true}
    boneToModify {boneName=dyn_furCollar_fr_r_01}
    constraintSetup [{targetBone {boneName=clavicle_r}}, {targetBone {boneName=neck_01}}]
    constraintWeights [1,0]
   }
   n46 = "Constraint - dyn_furCollar_fr_l_01 (PP)"!(ComponentPose=@n45, Alpha=0.1, ConstraintWeights_0=0.9, ConstraintWeights_1:float=0.1) -> (Pose) {
    alphaBoolBlend {blendOption=Linear}
    alphaScaleBiasClamp {bInterpResult=true}
    boneToModify {boneName=dyn_furCollar_fr_l_01}
    constraintSetup [{targetBone {boneName=clavicle_l}}, {targetBone {boneName=neck_01}}]
    constraintWeights [1,0]
   }
   n47 = "Constraint - dyn_furCollar_fr_r_03 (PP)"!(ComponentPose=@n46, Alpha=0.1, ConstraintWeights_0=1, ConstraintWeights_1:float=0) -> (Pose) {
    alphaBoolBlend {blendOption=Linear}
    alphaScaleBiasClamp {bInterpResult=true}
    boneToModify {boneName=dyn_furCollar_fr_r_03}
    constraintSetup [{targetBone {boneName=clavicle_r}}, {targetBone {boneName=neck_01}}]
    constraintWeights [1,0]
   }
   n48 = "Constraint - dyn_furCollar_fr_l_03 (PP)"!(ComponentPose=@n47, Alpha=0.1, ConstraintWeights_0=1, ConstraintWeights_1:float=0) -> (Pose) {
    alphaBoolBlend {blendOption=Linear}
    alphaScaleBiasClamp {bInterpResult=true}
    boneToModify {boneName=dyn_furCollar_fr_l_03}
    constraintSetup [{targetBone {boneName=clavicle_l}}, {targetBone {boneName=neck_01}}]
    constraintWeights [1,0]
   }
   n49 = "Transform (Modify) Bone - Bone: dyn_furCollar_fr_r_base"(ComponentPose=@n48, Alpha=1, Translation="-0.000,-0.000,-0.015", Rotation="0,0,0", Scale="1,1,1") {
    rotationMode=BMM_Additive
    scaleMode=BMM_Replace
    translationMode=BMM_Additive
    alphaBoolBlend {blendOption=Linear}
    alphaScaleBiasClamp {bInterpResult=true}
    boneToModify {boneName=dyn_furCollar_fr_r_base}
    translation=<0 0 -0.0149>
   }
   n50 = "Transform (Modify) Bone - Bone: dyn_furCollar_fr_r_01"(ComponentPose=@n49, Alpha=1, Translation="0,0,0", Rotation="0,0,0", Scale="1,1,1") {
    rotationMode=BMM_Additive
    scaleMode=BMM_Replace
    translationMode=BMM_Additive
    alphaBoolBlend {blendOption=Linear}
    alphaScaleBiasClamp {bInterpResult=true}
    boneToModify {boneName=dyn_furCollar_fr_r_01}
   }
   n51 = "Transform (Modify) Bone - Bone: dyn_furCollar_fr_r_03"(ComponentPose=@n50, Alpha=1, Translation="0,0,0", Rotation="0,0,0", Scale="1,1,1") {
    rotationMode=BMM_Additive
    scaleMode=BMM_Replace
    translationMode=BMM_Additive
    alphaBoolBlend {blendOption=Linear}
    alphaScaleBiasClamp {bInterpResult=true}
    boneToModify {boneName=dyn_furCollar_fr_r_03}
   }
   n52 = "Transform (Modify) Bone - Bone: dyn_furCollar_fr_r_05"(ComponentPose=@n51, Alpha=1, Translation="-0.000,0.000,-0.996", Rotation="0,0,0", Scale="1,1,0.75") {
    rotationMode=BMM_Additive
    scaleMode=BMM_Replace
    translationMode=BMM_Additive
    alphaBoolBlend {blendOption=Linear}
    alphaScaleBiasClamp {bInterpResult=true}
    boneToModify {boneName=dyn_furCollar_fr_r_05}
    scale {z=0.75}
    translation {z=-0.996}
   }
   n53 = "Transform (Modify) Bone - Bone: dyn_furCollar_bck_c_02"(ComponentPose=@n52, Alpha=1, Translation="0.000,2.393,-2.140", Rotation="1.292,4.830,-14.946", Scale="1,1,1") {
    rotationMode=BMM_Additive
    scaleMode=BMM_Replace
    translationMode=BMM_Additive
    alphaBoolBlend {blendOption=Linear}
    alphaScaleBiasClamp {bInterpResult=true}
    boneToModify {boneName=dyn_furCollar_bck_c_02}
    rotation=<rot 1.292 4.83 -14.946>
    translation {y=2.3931, z=-2.14}
   }
   n54 = "Transform (Modify) Bone - Bone: dyn_furCollar_bck_r_02"(ComponentPose=@n53, Alpha=1, Translation="0.000,-0.000,-1.338", Rotation="0.000,0.000,-20.000", Scale="1.000,1.000,0.750") {
    rotationMode=BMM_Additive
    scaleMode=BMM_Replace
    translationMode=BMM_Additive
    alphaBoolBlend {blendOption=Linear}
    alphaScaleBiasClamp {bInterpResult=true}
    boneToModify {boneName=dyn_furCollar_bck_r_02}
    rotation {roll=-20}
    scale {z=0.75}
    translation {z=-1.338}
   }
   n55 = "Transform (Modify) Bone - Bone: dyn_furCollar_bck_l_01"(ComponentPose=@n54, Alpha=1, Translation="-0.000,0.000,0.038", Rotation="0,0,0", Scale="1,1,1") {
    rotationMode=BMM_Additive
    scaleMode=BMM_Replace
    translationMode=BMM_Additive
    alphaBoolBlend {blendOption=Linear}
    alphaScaleBiasClamp {bInterpResult=true}
    boneToModify {boneName=dyn_furCollar_bck_l_01}
    translation {x=0, z=0.0385}
   }
   n56 = "Transform (Modify) Bone - Bone: dyn_furCollar_bck_l_02"(ComponentPose=@n55, Alpha=1, Translation="0.000,-0.000,-1.891", Rotation="0,0,0", Scale="1,1,1") {
    rotationMode=BMM_Additive
    scaleMode=BMM_Replace
    translationMode=BMM_Additive
    alphaBoolBlend {blendOption=Linear}
    alphaScaleBiasClamp {bInterpResult=true}
    boneToModify {boneName=dyn_furCollar_bck_l_02}
    translation=<0 0 -1.8913>
   }
   n57 = "Transform (Modify) Bone - Bone: dyn_furCollar_fr_r_base"(ComponentPose=@n56, Alpha=1, Translation="-0.210,0.340,0.705", Rotation="-35.000,0.000,-0.000", Scale="1.000,1.000,0.750") {
    rotationMode=BMM_Additive
    scaleMode=BMM_Replace
    translationMode=BMM_Additive
    alphaBoolBlend {blendOption=Linear}
    alphaScaleBiasClamp {bInterpResult=true}
    boneToModify {boneName=dyn_furCollar_fr_r_base}
    rotation {pitch=-35}
    scale {z=0.75}
    translation=<-0.21 0.34 0.7055>
   }
   n58 = "Transform (Modify) Bone - Bone: dyn_furCollar_fr_r_01"(ComponentPose=@n57, Alpha=1, Translation="0.000,1.318,-0.000", Rotation="-10.000,0.000,-0.000", Scale="1,1,1") {
    rotationMode=BMM_Additive
    scaleMode=BMM_Replace
    translationMode=BMM_Additive
    alphaBoolBlend {blendOption=Linear}
    alphaScaleBiasClamp {bInterpResult=true}
    boneToModify {boneName=dyn_furCollar_fr_r_01}
    rotation {pitch=-10}
    translation=<0 1.3184 0>
   }
   n59 = "Transform (Modify) Bone - Bone: dyn_furCollar_fr_r_03"(ComponentPose=@n58, Alpha=1, Translation="0.000,-0.000,0.476", Rotation="0,0,0", Scale="1,1,1") {
    rotationMode=BMM_Additive
    scaleMode=BMM_Replace
    translationMode=BMM_Additive
    alphaBoolBlend {blendOption=Linear}
    alphaScaleBiasClamp {bInterpResult=true}
    boneToModify {boneName=dyn_furCollar_fr_r_03}
    translation=<0 0 0.4758>
   }
   n60 = "Transform (Modify) Bone - Bone: dyn_furCollar_fr_r_04"(ComponentPose=@n59, Alpha=1, Translation="0.000,0.000,0.415", Rotation="0,0,0", Scale="1,1,1") {
    rotationMode=BMM_Additive
    scaleMode=BMM_Replace
    translationMode=BMM_Additive
    alphaBoolBlend {blendOption=Linear}
    alphaScaleBiasClamp {bInterpResult=true}
    boneToModify {boneName=dyn_furCollar_fr_r_04}
    translation {x=0, z=0.4147}
   }
   n61 = "Transform (Modify) Bone - Bone: dyn_furCollar_fr_r_05"(ComponentPose=@n60, Alpha=1, Translation="-2.526,-1.725,0.000", Rotation="9.656,-17.612,-15.221", Scale="1,1,1") {
    rotationMode=BMM_Additive
    scaleMode=BMM_Replace
    translationMode=BMM_Additive
    alphaBoolBlend {blendOption=Linear}
    alphaScaleBiasClamp {bInterpResult=true}
    boneToModify {boneName=dyn_furCollar_fr_r_05}
    rotation=<rot 9.656 -17.612 -15.221>
    translation=<-2.5263 -1.725 0>
   }
   n62 = "Transform (Modify) Bone - Bone: dyn_furCollar_fr_l_base"(ComponentPose=@n61, Alpha=1, Translation="-0.699,0.618,0.000", Rotation="30.000,0.000,0.000", Scale="1.000,1.000,0.750") {
    rotationMode=BMM_Additive
    scaleMode=BMM_Replace
    translationMode=BMM_Additive
    alphaBoolBlend {blendOption=Linear}
    alphaScaleBiasClamp {bInterpResult=true}
    boneToModify {boneName=dyn_furCollar_fr_l_base}
    rotation {pitch=30}
    scale {z=0.75}
    translation=<-0.699 0.618>
   }
   n63 = "Transform (Modify) Bone - Bone: dyn_furCollar_fr_l_01"(ComponentPose=@n62, Alpha=1, Translation="0.000,0.257,-0.000", Rotation="0,0,0", Scale="1,1,1") {
    rotationMode=BMM_Additive
    scaleMode=BMM_Replace
    translationMode=BMM_Additive
    alphaBoolBlend {blendOption=Linear}
    alphaScaleBiasClamp {bInterpResult=true}
    boneToModify {boneName=dyn_furCollar_fr_l_01}
    translation {y=0.257}
   }
   n64 = "Transform (Modify) Bone - Bone: dyn_furCollar_fr_l_03"(ComponentPose=@n63, Alpha=1, Translation="0.000,0.000,0.728", Rotation="0,0,0", Scale="1,1,1") {
    rotationMode=BMM_Additive
    scaleMode=BMM_Replace
    translationMode=BMM_Additive
    alphaBoolBlend {blendOption=Linear}
    alphaScaleBiasClamp {bInterpResult=true}
    boneToModify {boneName=dyn_furCollar_fr_l_03}
    translation {z=0.728}
   }
   n65 = "Transform (Modify) Bone - Bone: dyn_furCollar_fr_l_04"(ComponentPose=@n64, Alpha=1, Translation="1.554,0.000,0.882", Rotation="-10.000,0.000,-0.000", Scale="1,1,1") {
    rotationMode=BMM_Additive
    scaleMode=BMM_Replace
    translationMode=BMM_Additive
    alphaBoolBlend {blendOption=Linear}
    alphaScaleBiasClamp {bInterpResult=true}
    boneToModify {boneName=dyn_furCollar_fr_l_04}
    rotation {pitch=-10}
    translation {x=1.554, z=0.882}
   }
   n66 = "Transform (Modify) Bone - Bone: dyn_furCollar_fr_l_04"(ComponentPose=@n65, Alpha=1, Translation="0,-1,2", Rotation="0,0,0", Scale="1,1,1") {
    rotationMode=BMM_Additive
    scaleMode=BMM_Replace
    translationMode=BMM_Additive
    alphaBoolBlend {blendOption=Linear}
    alphaScaleBiasClamp {bInterpResult=true}
    boneToModify {boneName=dyn_furCollar_fr_l_04}
    translation {y=-1, z=2}
   }
   n67 = "Transform (Modify) Bone - Bone: dyn_furCollar_bck_l_02"(ComponentPose=@n66, Alpha=1, Translation="0.747,-1.000,0.000", Rotation="0,0,0", Scale="1,1,0.8") {
    rotationMode=BMM_Additive
    scaleMode=BMM_Replace
    translationMode=BMM_Additive
    alphaBoolBlend {blendOption=Linear}
    alphaScaleBiasClamp {bInterpResult=true}
    boneToModify {boneName=dyn_furCollar_bck_l_02}
    scale {z=0.8}
    translation=<0.747 -1>
   }
   n68 = "Transform (Modify) Bone - Bone: dyn_furCollar_bck_c_02"(ComponentPose=@n67, Alpha=1, Translation="-0.603,-2.000,0.909", Rotation="0,0,0", Scale="1,1,0.8") {
    rotationMode=BMM_Additive
    scaleMode=BMM_Replace
    translationMode=BMM_Additive
    alphaBoolBlend {blendOption=Linear}
    alphaScaleBiasClamp {bInterpResult=true}
    boneToModify {boneName=dyn_furCollar_bck_c_02}
    scale {z=0.8}
    translation=<-0.603 -2 0.9093>
   }
   n69 = "Transform (Modify) Bone - Bone: dyn_furCollar_bck_r_02"(ComponentPose=@n68, Alpha=1, Translation="0,0,0", Rotation="0,0,0", Scale="1,1,0.8") {
    rotationMode=BMM_Additive
    scaleMode=BMM_Replace
    translationMode=BMM_Additive
    alphaBoolBlend {blendOption=Linear}
    alphaScaleBiasClamp {bInterpResult=true}
    boneToModify {boneName=dyn_furCollar_bck_r_02}
    scale {z=0.8}
   }
   n70 = "Transform (Modify) Bone - Bone: dyn_furCollar_fr_r_05"(ComponentPose=@n69, Alpha=1, Translation="2.545,3.736,2.078", Rotation="0,0,0", Scale="1,1,0.8") {
    rotationMode=BMM_Additive
    scaleMode=BMM_Replace
    translationMode=BMM_Additive
    alphaBoolBlend {blendOption=Linear}
    alphaScaleBiasClamp {bInterpResult=true}
    boneToModify {boneName=dyn_furCollar_fr_r_05}
    scale {z=0.8}
    translation=<2.545 3.7358 2.078>
   }
   n71 = "Transform (Modify) Bone - Bone: dyn_furCollar_fr_r_04"(ComponentPose=@n70, Alpha=1, Translation="0,0,0", Rotation="0,0,0", Scale="1,1,1") {
    rotationMode=BMM_Additive
    scaleMode=BMM_Replace
    translationMode=BMM_Additive
    alphaBoolBlend {blendOption=Linear}
    alphaScaleBiasClamp {bInterpResult=true}
    boneToModify {boneName=dyn_furCollar_fr_r_04}
   }
   n72 = "Transform (Modify) Bone - Bone: dyn_furCollar_fr_r_03"(ComponentPose=@n71, Alpha=1, Translation="0,0,0", Rotation="0,0,0", Scale="1,1,1") {
    rotationMode=BMM_Additive
    scaleMode=BMM_Replace
    translationMode=BMM_Additive
    alphaBoolBlend {blendOption=Linear}
    alphaScaleBiasClamp {bInterpResult=true}
    boneToModify {boneName=dyn_furCollar_fr_r_03}
   }
   n73 = "Transform (Modify) Bone - Bone: dyn_furCollar_bck_c_02"(ComponentPose=@n72, Alpha=1, Translation="0.000,-1.108,0.000", Rotation="0,0,0", Scale="1,1,1") {
    rotationMode=BMM_Additive
    scaleMode=BMM_Replace
    translationMode=BMM_Additive
    alphaBoolBlend {blendOption=Linear}
    alphaScaleBiasClamp {bInterpResult=true}
    boneToModify {boneName=dyn_furCollar_bck_c_02}
    translation {y=-1.108}
   }
   n74 = "Transform (Modify) Bone - Bone: dyn_furCollar_fr_r_05"(ComponentPose=@n73, Alpha=1, Translation="-0.000,-3.706,-1.157", Rotation="0.000,0.000,-5.000", Scale="1,1,0.87") {
    rotationMode=BMM_Additive
    scaleMode=BMM_Replace
    translationMode=BMM_Additive
    alphaBoolBlend {blendOption=Linear}
    alphaScaleBiasClamp {bInterpResult=true}
    boneToModify {boneName=dyn_furCollar_fr_r_05}
    rotation {roll=-5}
    scale {z=0.87}
    translation {y=-3.706, z=-1.157}
   }
   n75 = "Transform (Modify) Bone - Bone: dyn_furCollar_fr_l_04"(ComponentPose=@n74, Alpha=1, Translation="0,0,-0.3", Rotation="0.000,0.000,-5.000", Scale="1,1,0.87") {
    rotationMode=BMM_Additive
    scaleMode=BMM_Replace
    translationMode=BMM_Additive
    alphaBoolBlend {blendOption=Linear}
    alphaScaleBiasClamp {bInterpResult=true}
    boneToModify {boneName=dyn_furCollar_fr_l_04}
    rotation {roll=-5}
    scale {z=0.87}
    translation {z=-0.3}
   }
   n76 = "Transform (Modify) Bone - Bone: dyn_furCollar_fr_r_05"(ComponentPose=@n75, Alpha=1, Translation="-0.000,-3.706,-1.157", Rotation="0.000,0.000,-5.000", Scale="1,1,0.87") {
    rotationMode=BMM_Additive
    scaleMode=BMM_Replace
    translationMode=BMM_Additive
    alphaBoolBlend {blendOption=Linear}
    alphaScaleBiasClamp {bInterpResult=true}
    boneToModify {boneName=dyn_furCollar_fr_r_05}
    rotation {roll=-5}
    scale {z=0.87}
    translation {y=-3.706, z=-1.157}
   }
   n77 = "Transform (Modify) Bone - Bone: dyn_furCollar_fr_l_04"(ComponentPose=@n76, Alpha=1, Translation="0,0,-0.3", Rotation="0.000,0.000,-5.000", Scale="1,1,0.87") {
    rotationMode=BMM_Additive
    scaleMode=BMM_Replace
    translationMode=BMM_Additive
    alphaBoolBlend {blendOption=Linear}
    alphaScaleBiasClamp {bInterpResult=true}
    boneToModify {boneName=dyn_furCollar_fr_l_04}
    rotation {roll=-5}
    scale {z=0.87}
    translation {z=-0.3}
   }
   n78 = "Transform (Modify) Bone - Bone: dyn_furCollar_fr_r_05"(ComponentPose=@n77, Alpha=1, Translation="-1.407,0.732,0.000", Rotation="-0.000,-20.000,0.000", Scale="1,1,1") {
    rotationMode=BMM_Additive
    scaleMode=BMM_Replace
    translationMode=BMM_Additive
    alphaBoolBlend {blendOption=Linear}
    alphaScaleBiasClamp {bInterpResult=true}
    boneToModify {boneName=dyn_furCollar_fr_r_05}
    rotation {yaw=-20}
    translation=<-1.407 0.732>
   }
   n79 = "Transform (Modify) Bone - Bone: dyn_furCollar_fr_r_04"(ComponentPose=@n78, Alpha=1, Translation="0.000,-1.718,-0.019", Rotation="0,0,0", Scale="1,1,1") {
    rotationMode=BMM_Additive
    scaleMode=BMM_Replace
    translationMode=BMM_Additive
    alphaBoolBlend {blendOption=Linear}
    alphaScaleBiasClamp {bInterpResult=true}
    boneToModify {boneName=dyn_furCollar_fr_r_04}
    translation {y=-1.718, z=-0.019}
   }
   n80 = "Transform (Modify) Bone - Bone: dyn_furCollar_fr_r_03"(ComponentPose=@n79, Alpha=1, Translation="0.000,-2.127,2.396", Rotation="0,0,0", Scale="1,1,1") {
    rotationMode=BMM_Additive
    scaleMode=BMM_Replace
    translationMode=BMM_Additive
    alphaBoolBlend {blendOption=Linear}
    alphaScaleBiasClamp {bInterpResult=true}
    boneToModify {boneName=dyn_furCollar_fr_r_03}
    translation {y=-2.127, z=2.396}
   }
   n81 = "Transform (Modify) Bone - Bone: dyn_furCollar_bck_r_01"(ComponentPose=@n80, Alpha=1, Translation="1.239,0.000,0.000", Rotation="0,0,0", Scale="1,1,1") {
    rotationMode=BMM_Additive
    scaleMode=BMM_Replace
    translationMode=BMM_Additive
    alphaBoolBlend {blendOption=Linear}
    alphaScaleBiasClamp {bInterpResult=true}
    boneToModify {boneName=dyn_furCollar_bck_r_01}
    translation {x=1.239}
   }
   n82 = "Transform (Modify) Bone - Bone: dyn_furCollar_fr_r_01"(ComponentPose=@n81, Alpha=1, Translation="-0.000,1.704,0.000", Rotation="0,0,0", Scale="1,1,1") {
    rotationMode=BMM_Additive
    scaleMode=BMM_Replace
    translationMode=BMM_Additive
    alphaBoolBlend {blendOption=Linear}
    alphaScaleBiasClamp {bInterpResult=true}
    boneToModify {boneName=dyn_furCollar_fr_r_01}
    translation {y=1.704}
   }
   n83 = "Transform (Modify) Bone - Bone: dyn_furCollar_bck_r_01"(ComponentPose=@n82, Alpha=1, Translation="0,0,0", Rotation="0,0,0", Scale="1,1,1") {
    rotationMode=BMM_Additive
    scaleMode=BMM_Replace
    translationMode=BMM_Additive
    alphaBoolBlend {blendOption=Linear}
    alphaScaleBiasClamp {bInterpResult=true}
    boneToModify {boneName=dyn_furCollar_bck_r_01}
   }
   n84 = "Transform (Modify) Bone - Bone: dyn_furCollar_fr_r_03"(ComponentPose=@n83, Alpha=1, Translation="0,0,0", Rotation="0,0,0", Scale="1,1,1") {
    rotationMode=BMM_Additive
    scaleMode=BMM_Replace
    translationMode=BMM_Additive
    alphaBoolBlend {blendOption=Linear}
    alphaScaleBiasClamp {bInterpResult=true}
    boneToModify {boneName=dyn_furCollar_fr_r_03}
   }
   n85 = "Transform (Modify) Bone - Bone: dyn_furCollar_fr_r_05"(ComponentPose=@n84, Alpha=1, Translation="1.535,-3.420,1.103", Rotation="0,0,0", Scale="1,1,1") {
    rotationMode=BMM_Additive
    scaleMode=BMM_Replace
    translationMode=BMM_Additive
    alphaBoolBlend {blendOption=Linear}
    alphaScaleBiasClamp {bInterpResult=true}
    boneToModify {boneName=dyn_furCollar_fr_r_05}
    translation=<1.535 -3.42 1.103>
   }
   n86 = "Transform (Modify) Bone - Bone: dyn_furCollar_fr_l_02"(ComponentPose=@n85, Alpha=1, Translation="-0.000,-0.750,2.747", Rotation="0,0,0", Scale="1,1,1") {
    rotationMode=BMM_Additive
    scaleMode=BMM_Replace
    translationMode=BMM_Additive
    alphaBoolBlend {blendOption=Linear}
    alphaScaleBiasClamp {bInterpResult=true}
    boneToModify {boneName=dyn_furCollar_fr_l_02}
    translation {y=-0.75, z=2.747}
   }
   n87 = "Transform (Modify) Bone - Bone: dyn_furCollar_fr_l_03"(ComponentPose=@n86, Alpha=1, Translation="2.112,-0.000,0.332", Rotation="0,0,0", Scale="1,1,1") {
    rotationMode=BMM_Additive
    scaleMode=BMM_Replace
    translationMode=BMM_Additive
    alphaBoolBlend {blendOption=Linear}
    alphaScaleBiasClamp {bInterpResult=true}
    boneToModify {boneName=dyn_furCollar_fr_l_03}
    translation {x=2.112, z=0.332}
   }
   n88 = "Transform (Modify) Bone - Bone: dyn_furCollar_fr_r_base"(ComponentPose=@n87, Alpha=1, Translation="0.000,0.000,0.425", Rotation="0,0,0", Scale="1,1,1") {
    rotationMode=BMM_Additive
    scaleMode=BMM_Replace
    translationMode=BMM_Additive
    alphaBoolBlend {blendOption=Linear}
    alphaScaleBiasClamp {bInterpResult=true}
    boneToModify {boneName=dyn_furCollar_fr_r_base}
    translation=<0 0 0.4252>
   }
   n89 = "Transform (Modify) Bone - Bone: dyn_furCollar_fr_r_03"(ComponentPose=@n88, Alpha=1, Translation="-0.000,0.747,-2.450", Rotation="0,0,0", Scale="1.000,1.000,0.750") {
    rotationMode=BMM_Additive
    scaleMode=BMM_Replace
    translationMode=BMM_Additive
    alphaBoolBlend {blendOption=Linear}
    alphaScaleBiasClamp {bInterpResult=true}
    boneToModify {boneName=dyn_furCollar_fr_r_03}
    scale {z=0.75}
    translation=<0 0.747 -2.4498>
   }
   n90 = "Transform (Modify) Bone - Bone: dyn_furCollar_fr_r_05"(ComponentPose=@n89, Alpha=1, Translation="-0.000,0.000,-1.829", Rotation="0,0,0", Scale="1.000,1.000,0.750") {
    rotationMode=BMM_Additive
    scaleMode=BMM_Replace
    translationMode=BMM_Additive
    alphaBoolBlend {blendOption=Linear}
    alphaScaleBiasClamp {bInterpResult=true}
    boneToModify {boneName=dyn_furCollar_fr_r_05}
    scale {z=0.75}
    translation {z=-1.829}
   }
   n91 = "Transform (Modify) Bone - Bone: dyn_furCollar_fr_r_04"(ComponentPose=@n90, Alpha=1, Translation="1.553,-1.308,-1", Rotation="-20.000,0.000,-0.000", Scale="0.500,1.000,1.000") {
    rotationMode=BMM_Additive
    scaleMode=BMM_Replace
    translationMode=BMM_Additive
    alphaBoolBlend {blendOption=Linear}
    alphaScaleBiasClamp {bInterpResult=true}
    boneToModify {boneName=dyn_furCollar_fr_r_04}
    rotation {pitch=-20}
    scale {x=0.5}
    translation=<1.553 -1.308 -1>
   }
   n92 = "Transform (Modify) Bone - Bone: dyn_furCollar_fr_r_05"(ComponentPose=@n91, Alpha=1, Translation="1.299,0.392,-1.743", Rotation="0,0,0", Scale="1,1,0.6") {
    rotationMode=BMM_Additive
    scaleMode=BMM_Replace
    translationMode=BMM_Additive
    alphaBoolBlend {blendOption=Linear}
    alphaScaleBiasClamp {bInterpResult=true}
    boneToModify {boneName=dyn_furCollar_fr_r_05}
    scale {z=0.6}
    translation=<1.299 0.392 -1.743>
   }
   n93 = "Transform (Modify) Bone - Bone: dyn_furCollar_bck_r_01"(ComponentPose=@n92, Alpha=1, Translation="-0.000,0.000,-0.740", Rotation="0,0,0", Scale="1,1,1") {
    rotationMode=BMM_Additive
    scaleMode=BMM_Replace
    translationMode=BMM_Additive
    alphaBoolBlend {blendOption=Linear}
    alphaScaleBiasClamp {bInterpResult=true}
    boneToModify {boneName=dyn_furCollar_bck_r_01}
    translation=<0 0 -0.7403>
   }
   n94 = "Transform (Modify) Bone - Bone: dyn_furCollar_fr_r_04"(ComponentPose=@n93, Alpha=1, Translation="0.000,-0.000,0.465", Rotation="0,0,0", Scale="1,1,1") {
    rotationMode=BMM_Additive
    scaleMode=BMM_Replace
    translationMode=BMM_Additive
    alphaBoolBlend {blendOption=Linear}
    alphaScaleBiasClamp {bInterpResult=true}
    boneToModify {boneName=dyn_furCollar_fr_r_04}
    translation {z=0.465}
   }
   n95 = "Transform (Modify) Bone - Bone: dyn_furCollar_fr_r_05"(ComponentPose=@n94, Alpha=1, Translation="0.000,-0.822,3.086", Rotation="0,0,0", Scale="1,1,1") {
    rotationMode=BMM_Additive
    scaleMode=BMM_Replace
    translationMode=BMM_Additive
    alphaBoolBlend {blendOption=Linear}
    alphaScaleBiasClamp {bInterpResult=true}
    boneToModify {boneName=dyn_furCollar_fr_r_05}
    translation=<0 -0.8217 3.086>
   }
   n96 = "Constraint - dyn_chestStrap_fr_l_base (PPPPPPP)"!(
    ComponentPose=@n95
    Alpha=1
    ConstraintWeights_0=0.078
    ConstraintWeights_1:float=0.8725
    ConstraintWeights_2:float=0.0008
    ConstraintWeights_3:float=0.0282
    ConstraintWeights_4:float=0.0057
    ConstraintWeights_5:float=0.0052
    ConstraintWeights_6:float=0.0096
   ) -> (Pose) {
    alphaBoolBlend {blendOption=Linear}
    boneToModify {boneName=dyn_chestStrap_fr_l_base}
    constraintSetup [
     {targetBone {boneName=neck_01}}
     {targetBone {boneName=spine_05}}
     {targetBone {boneName=spine_04}}
     {targetBone {boneName=upperarm_twist_01_l}}
     {targetBone {boneName=upperarm_l}}
     {targetBone {boneName=clavicle_l}}
     {targetBone {boneName=spine_03}}
    ]
    constraintWeights [1,1,1,1,1,1,1]
   }
   n97 = "Constraint - dyn_chestStrap_bck_l_base (P)"!(ComponentPose=@n96, Alpha=1, ConstraintWeights_0=1, ConstraintWeights_1:float=0) -> (Pose) {
    alphaBoolBlend {blendOption=Linear}
    boneToModify {boneName=dyn_chestStrap_bck_l_base}
    constraintSetup [{targetBone {boneName=spine_05}}]
    constraintWeights [1,0]
   }
   n98 = "Constraint - dyn_chestStrap_fr_r_base (PPPPPPP)"!(
    ComponentPose=@n97
    Alpha=1
    ConstraintWeights_0=0.0107
    ConstraintWeights_1:float=0.0332
    ConstraintWeights_2:float=0.9465
    ConstraintWeights_3:float=0.0006
    ConstraintWeights_4:float=0.0016
    ConstraintWeights_5:float=0.0011
    ConstraintWeights_6:float=0.0062
   ) -> (Pose) {
    alphaBoolBlend {blendOption=Linear}
    boneToModify {boneName=dyn_chestStrap_fr_r_base}
    constraintSetup [
     {targetBone {boneName=upperarm_twist_01_r}}
     {targetBone {boneName=neck_01}}
     {targetBone {boneName=spine_05}}
     {targetBone {boneName=spine_04}}
     {targetBone {boneName=clavicle_r}}
     {targetBone {boneName=upperarm_r}}
     {targetBone {boneName=spine_03}}
    ]
    constraintWeights [1,1,1,1,1,1,1]
   }
   n99 = "Constraint - dyn_chestStrap_bck_r_base (P)"!(ComponentPose=@n98, Alpha=1, ConstraintWeights_0=1, ConstraintWeights_1:float=0) -> (Pose) {
    alphaBoolBlend {blendOption=Linear}
    boneToModify {boneName=dyn_chestStrap_bck_r_base}
    constraintSetup [{targetBone {boneName=spine_05}}]
    constraintWeights [1,0]
   }
   n100 = RigidBody(ComponentPose=@n99, OverrideWorldGravity="0,0,0") {
    bClampLinearTranslationLimitToRefPose=true
    bOverrideWorldGravity=true
    overridePhysicsAsset=&/F_MED_VitalPsych_Claws_Strap_Physics
    simulationSpace=BaseBoneSpace
    alphaBoolBlend {blendOption=Linear}
    alphaScaleBiasClamp {bInterpResult=true}
    baseBoneRef {boneName=spine_05}
    componentAppliedLinearAccClamp=<3000 3000 3000>
    componentLinearAccScale=<0.6 0.6 0.7>
    componentLinearVelScale=<0.2 0.2 0.2>
   }
   n101 = "Transform (Modify) Bone - Bone: dyn_chestStrap_fr_r_base"(ComponentPose=@n100, Alpha=1, Translation="-0.633,-0.514,0", Rotation="0,0,0", Scale="1,1,1") {
    rotationMode=BMM_Additive
    scaleMode=BMM_Replace
    translationMode=BMM_Additive
    alphaBoolBlend {blendOption=Linear}
    alphaScaleBiasClamp {bInterpResult=true}
    boneToModify {boneName=dyn_chestStrap_fr_r_base}
    translation=<-0.633 -0.514>
   }
   n102 = "Transform (Modify) Bone - Bone: dyn_chestStrap_fr_r_base"(ComponentPose=@n101, Alpha=1, Translation="1.504,4.778,0.000", Rotation="-10.000,0.000,-0.000", Scale="1,1,1") {
    rotationMode=BMM_Additive
    scaleMode=BMM_Replace
    translationMode=BMM_Additive
    alphaBoolBlend {blendOption=Linear}
    alphaScaleBiasClamp {bInterpResult=true}
    boneToModify {boneName=dyn_chestStrap_fr_r_base}
    rotation {pitch=-10}
    translation=<1.5037 4.778 0>
   }
   n103 = Anim Dynamics - Bone: dyn_pouch!(ComponentPose=@n102, Alpha=1, GravityOverride:Vector="0,0,0", ExternalForce="0,0,0") -> (Pose) {
    bGravityOverrideInSimSpace=true
    bUseGravityOverride=true
    alphaBoolBlend {blendOption=Linear}
    boundBone {boneName=dyn_pouch}
    componentAppliedLinearAccClamp=<10000 10000 10000>
    componentLinearAccScale=<0.3 0.3 0.3>
    physicsBodyDefinitions [{
      boundBone {boneName=dyn_pouch}
      boxExtents=<15 15 15>
      constraintSetup {angularLimitsMax=<20 0 0>, angularLimitsMin=<-20 0 0>, angularTarget=<0 0 0>}
      localJointOffset=<0 0.05 0>
     }]
    relativeSpaceBone {boneName=pelvis}
   }
   n104 = Layered blend per bone(
    BasePose=Copy Pose From Mesh() {bUseAttachedParent=true}
    BlendPoses_0="Sequence Evaluator (by time) 'VitalPsychClaws_Fur_LobbyIdle'"!(ExplicitTime:float=Get Fur Position() -> (Fur Position:double)) -> (Pose:Pose Link Structure) {
     is=title AnimGraphNode_SequenceEvaluator
     sequence=/BRCosmetics/Animation/Game/MainPlayer/Menu/BR/CustomSkins/_s24/VitalPsychF/VitalPsychClaws_Fur_LobbyIdle
    }
   ) {layerSetup [{branchFilters [
       {boneName=spine_05}
       {boneName=dyn_hair_l_fr_02}
       {boneName=dyn_hair_l_fr_side_01}
       {boneName=dyn_hair_top}
       {boneName=dyn_hair_r_fr_01}
       {boneName=dyn_hair_fr_c_02}
       {boneName=dyn_hair_fr_c_01}
       {boneName=dyn_hair_fr_c_03}
      ]}]}
   Output Pose(Result=Blend Poses by bool(BlendPose_0=@n104, BlendPose_1=Component To Local(ComponentPose=@n103), bActiveValue=Get PlayingFrontendIdle() -> (PlayingFrontendIdle:bool)) {blendType=Linear}) {layerGroup=None}
  ]}
 EventGraph {nodes [
   Event BlueprintUpdateAnimation()
   > n1 = Sequence!(execute) -> (then_0, then_1, then_2:Exec)
   > Set anim alpha(anim alpha:double=1) -> (Output_Get:double)
   > n2 = Gravity Override State Function DEPRECATED(CharacterPartAnimInstance=Self-Reference(), GameTimeInSeconds=GetGameTimeInSeconds(), GravityOverrideParams=Get Gravity Override Params())
   > Set Gravity Override(Gravity Override=@n2)
   n3 = GetOwnerSkeletalMeshComponent()
   n4 = BreakRotator(InRot=BreakTransform(InTransform=GetDeltaTransformFromRefPose(self=@n3, BoneName=head, BaseName=spine_05)).Rotation)
   > Set Head Up(Head Up:double=MapRangeClamped(Value=@n4.Yaw, InRangeA=-20, InRangeB=-41)) -> (Output_Get:double)
   n5 = "To Float (Boolean)"(InBool=Get bIsTargeting())
   > n6 = Set Head Up Targeting(Head Up Targeting:double=float * float(A=MapRangeClamped(Value=@n4.Yaw, InRangeA=0, InRangeB=-17), B=@n5) {is=title K2Node_PromotableOperator}) -> (Output_Get:double)
   n7 = Get bIsCrouching()
   > Set crouch targeting up(crouch targeting up:double=float * float(A="To Float (Boolean)"(InBool=@n7), B=@n6)) -> (Output_Get:double)
   > Set Head Down Targeting(Head Down Targeting:double=float * float(A=MapRangeClamped(Value=@n4.Yaw, InRangeA=1, InRangeB=18), B=@n5)) -> (Output_Get:double)
   > n8 = Branch(Condition=Get bIsParachuteOpened())
   > n9 = Set parachute alpha(parachute alpha:double=1) -> (Output_Get:double)
   Set Arm Up R(execute=[@n10.then, @n9.then], Arm Up R:double=0) -> (Output_Get:double)
   > Set Arm Side R(Arm Side R:double=0) -> (Output_Get:double)
   > Set Arm Up L(Arm Up L:double=0) -> (Output_Get:double)
   > Set anim alpha(anim alpha=0.3)
   Set parachute alpha(execute=@n8.else, parachute alpha=0)
   > n10 = Branch(Condition=OR Boolean(A=Get bIsSkydiveFloating(), B=Get bIsSkydiveDiveMode()))
   n11 = Branch(execute=@n10.else, Condition=Get bIsPlayingEmote())
   > Set emote alpha(emote alpha:double=0.65) -> (Output_Get:double)
   > Set anim alpha(anim alpha=0)
   Set emote alpha(execute=@n11.else, emote alpha=0)
   n12 = GetOwnerSkeletalMeshComponent()
   n13 = BreakRotator(InRot=BreakTransform(InTransform=GetDeltaTransformFromRefPose(self=@n12, BoneName=upperarm_r, BaseName=spine_05)).Rotation)
   > n14 = Set Arm Up R(Arm Up R=MapRangeClamped(Value="Absolute (Float)"(A=@n13.Yaw), InRangeA=35, InRangeB=100))
   > Set Arm Side R(Arm Side R=MapRangeClamped(Value="Absolute (Float)"(A=@n13.Pitch), InRangeA=25, InRangeB=50))
   n15 = Get bIsMoving2D()
   > Set arm back R(arm back R:double=float * float(A=MapRangeClamped(Value=@n13.Yaw, InRangeA=0, InRangeB=-5), B="To Float (Boolean)"(InBool=@n15))) -> (Output_Get:double)
   n16 = "To Float (Boolean)"(InBool="float <= float"(A=@n14, B=0.5) -> (ReturnValue=~))
   > n17 = Branch(Condition="float > float"(A=float * float(A=@n6, B=@n16)))
   > Set strap rban alpha(strap rban alpha:double=0) -> (Output_Get:double)
   > n18 = Set strap mover(strap mover:double=1) -> (Output_Get:double)
   n19 = GetOwnerSkeletalMeshComponent()
   n20 = BreakRotator(InRot=BreakTransform(InTransform=GetDeltaTransformFromRefPose(self=@n19, BoneName=upperarm_l, BaseName=spine_05)).Rotation)
   Set Arm Up L(execute=[@n35.else, @n36.then, @n18.then], Arm Up L=MapRangeClamped(Value="Absolute (Float)"(A=@n20.Yaw), InRangeA=30, InRangeB=90))
   > n21 = Branch(Condition=Get bIsSkydiveFloating())
   > Set Arm Up L(Arm Up L=0)
   > n22 = Set Arm Up R(Arm Up R=0)
   n23 = Get bIsJumping()
   n24 = Get bIsFalling()
   n25 = Get Arm Up R() -> (Arm Up R:double)
   n26 = Branch(execute=[@n21.else, @n22.then], Condition=AND Boolean(A=OR Boolean(A=@n23, B=@n24), B="float >= float"(A=@n25, B=0.6)))
   > n27 = Set jump arm up(jump arm up:double=1) -> (Output_Get:double)
   n28 = Get bIsCrouching()
   n29 = Get bIsMoving2D()
   n30 = Get Head Up Targeting() -> (Head Up Targeting:double)
   n31 = Branch(execute=[@n32.then, @n27.then], Condition=AND Boolean(A=AND Boolean(A=@n28, B=@n29), B="float >= float"(A=@n30, B=0.8)))
   > Set crouch walk target down(crouch walk target down:double=1) -> (Output_Get:double)
   Set crouch walk target down(execute=@n31.else, crouch walk target down=0)
   n32 = Set jump arm up(execute=@n26.else, jump arm up=0)
   Set strap rban alpha(execute=@n17.else, strap rban alpha=1)
   > Set strap mover(strap mover=0)
   > n33 = Branch(Condition="float >= float"(A=Get Head Down Targeting() -> (Head Down Targeting:double), B=0.8))
   > Set strap rban alpha(strap rban alpha=0)
   > n34 = Set strap mover(strap mover=1)
   n35 = Branch(execute=[@n34.then, @n37.then], Condition="float >= float"(A=Get crouch targeting up() -> (crouch targeting up:double), B=1))
   > n36 = Set strap rban alpha(strap rban alpha=0.8)
   Set strap rban alpha(execute=@n33.else, strap rban alpha=1)
   > n37 = Set strap mover(strap mover=0)
   n38 = GetAnimInstance(self=GetOwnerSkeletalMeshComponent())
   Set PlayingFrontendIdle(
    execute=[@n1.then_1, @n40.then_0]
    PlayingFrontendIdle:bool="float > float"(A=GetCurveValue(self=@n38, CurveName=FrontEnd_Idle_Bypass), B=0) {is=title K2Node_CallFunction}
   ) -> (Output_Get:bool)
   n39 = GetOwnerSkeletalMeshComponent()
   Set Fur Position(
    execute=@n1.then_2
    Fur Position:double=Montage_GetPosition!(
     self:Anim Instance Object Reference=GetAnimInstance(self=@n39)
     Montage:Anim Montage Object Reference=/BRCosmetics/Animation/Game/MainPlayer/Menu/BR/CustomSkins/_s24/VitalPsychF/VitalPsychClaws_Female_Idle_Montage
    ) -> (ReturnValue:float)
   ) -> (Output_Get:double)
   K2Node_Event_1 = CustomEvent!() -> (OutputDelegate, then, DeltaTimeX:float=~) {is=title K2Node_CustomEvent}
   > n40 = Sequence!(execute) -> (then_0, then_1)
   Get bIsSkyDiving()
  ]}
]
variables [
 anim alpha {type=number}
 Anim_Inst {type=AnimInstance, default=null}
 arm back R {type=number}
 Arm Side R {type=number}
 Arm Up L {type=number}
 Arm Up R {type=number}
 crouch targeting up {type=number}
 crouch walk target down {type=number}
 emote alpha {type=number}
 Fur Position {type=number}
 Gravity Override {type=Vector}
 Gravity Override Params {type=GravityOverrideParamsStruct}
 Head Down Targeting {type=number}
 Head Up {type=number}
 Head Up Targeting {type=number}
 jump arm up {type=number}
 parachute alpha {type=number}
 PlayingFrontendIdle {type=boolean}
 strap mover {type=number}
 strap rban alpha {type=number}
]
